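# Material function for the TurboFueled character's backpack, adding animated, glass-tinted emissive to an input material where the tech-art mask selects it.
# Two scrolling noise layers at differing speeds combine with a projected swirl flow map, using pre-skinned positions.
# Each layer pairs with a view-dependent fresnel term and gradient-texture lookup, and the results average together.
# Glass input opacity (MF_Glass preview) controls the tint blend before the masked emissive override.
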MaterialFunction /BRCosmetics/Characters/Player/Male/Medium/Bodies/M_MED_TurboFueled/FX/MF_TurboFueled_Backpack
graph {nodes [
  n1 = NamedRerouteUsage()
  n2 = NamedRerouteUsage()
  n3 = VectorParameter() {parameterName=GlassColour, defaultValue=<rgba 0.3325 0.0194 0.0021 1>}
  n4 = GetMaterialAttributes(None=NamedRerouteUsage()) {attributeGetTypes [#1]}
  n5 = LinearInterpolate(
   A=Multiply(A=Multiply(A=@n2, B=@n3.RGB), B=OneMinus(None=@n4.Opacity))
   B=NamedRerouteUsage()
   Alpha=GetMaterialAttributes(None=NamedRerouteUsage()) {attributeGetTypes [#1]}.Opacity
  ) {constA=0.1, constB=0.1}
  n6 = DotProduct(A=NamedRerouteUsage(), B=VectorParameter() {group=FlameMaterial, parameterName=TechArtMaskEmissiveChannel, defaultValue=<rgba 0 1 0 0>}.RGB)
  FunctionOutput(None=SetMaterialAttributes!(MaterialAttributes=NamedRerouteUsage(), EmissiveColor=LinearInterpolate(A=@n1, B=@n5, Alpha=@n6) {constA=0.1}) {attributeSetTypes [#2]})
  n7 = TextureObject() {samplerType=SAMPLERTYPE_Normal, texture=/Game/Global/Textures/Default/Flowmaps/T_SwirlFlow_02_N}
  n8 = ScalarParameter() {defaultValue=40, group=BucketKick, parameterName=DistortionFlowTextureSize}
  n9 = LocalPosition() {includedOffsets=ExcludeOffsets, localOrigin=InstancePreSkinning}
  n10 = Add(
   A=NamedRerouteDeclaration(None=VertexInterpolator(VS=@n9.XYZ).PS) {name=PreskinnedPos, nodeColor=<rgba 0 0.9766 1 1>}
   B=Multiply(A=Time(), B=VectorParameter() {group=BucketKick, parameterName=DistortionFlowSpeed, defaultValue=<rgba 0.5 0.5 -12 1>}.RGB)
  )
  n11 = PreSkinnedNormal()
  n12 = Subtract(
   A=NamedRerouteDeclaration(None=VertexInterpolator(VS=@n11).PS) {name=PreskinnedNormal, nodeColor=<rgba 1 0 0.6562 1>}
   B=Constant3Vector() {constant=<rgba -100 0 0 0>}
  )
  n13 = NamedRerouteUsage()
  NamedRerouteDeclaration(None=Multiply(
    A=MaterialFunctionCall(TextureObject=@n7, TextureSize=@n8, WorldPosition=@n10, World Space Normal=@n12, ProjectionTransitionContrast=@n13) {materialFunction=&WorldAlignedTexture}.XYZ Texture
    B=ScalarParameter() {defaultValue=1, group=BucketKick, parameterName=DistortionFlowStrength, sortPriority=20}
   )) {name=Flow, nodeColor=<rgba 0.3594 0 1 1>}
  n14 = Power(Base=NamedRerouteUsage(), Exp=ScalarParameter() {defaultValue=0.5, group=BucketKick, parameterName=DistortionNoiseExponent, sortPriority=14})
  NamedRerouteDeclaration(None=Multiply(
    A=ConstantBiasScale(None=@n14) {bias=-0.5, scale=2}
    B=ScalarParameter() {defaultValue=0.2, group=BucketKick, parameterName=DistortionNoiseStrength, sortPriority=10}
   )) {name=Distortion, nodeColor=<rgba 1 0.7031 0 1>}
  n15 = NamedRerouteUsage()
  n16 = DotProduct(A=CameraVectorWS(), B=PixelNormalWS())
  n17 = NamedRerouteDeclaration(None=Power(Base=Saturate(None=@n16), Exp=ScalarParameter() {defaultValue=3, group=BucketKick, parameterName=FresnelExponent, sortPriority=43})) {name=FresenlValue, nodeColor=<rgba 0.7344 0 1 1>}
  NamedRerouteDeclaration(None=Saturate(None=Add(A=@n15, B=@n17))) {name=GradientUVX, nodeColor=<rgba 0 1 0.0625 1>}
  n18 = VertexInterpolator(VS=LocalPosition() {includedOffsets=ExcludeOffsets, localOrigin=InstancePreSkinning}.XYZ)
  n19 = NamedRerouteUsage()
  n20 = Time() {period=2}
  n21 = VectorParameter() {group=BucketKick, parameterName=DistortionNoiseSpeed, sortPriority=12, defaultValue=<rgba 0 0 -11 0>}
  n22 = Add(
   A=Add(A=LinearInterpolate(A=@n18.PS, B=@n19) {constAlpha=0.1}, B=Multiply(A=@n20, B=@n21.RGB))
   B=Add(A=NamedRerouteUsage(), B=NamedRerouteUsage())
  )
  n23 = TextureObject() {samplerType=SAMPLERTYPE_LinearColor, texture=&/T_TurboFueled_Body_FX}
  n24 = TextureSample(UVs=ComponentMask(None=Multiply(B=@n22) {constA=0.05}) {b=true, r=true}, Tex=@n23)
  NamedRerouteDeclaration(None=Saturate(None=ComponentMask(None=@n24.RGB) {b=true, g=true})) {name=DistortionNoiseSample, nodeColor=<rgba 1 0 0.5391 1>}
  NamedRerouteDeclaration(None=ScalarParameter() {defaultValue=0.5, group=BucketKick, parameterName=DistortionNoiseTransitionContrast, sortPriority=13}) {name=TransitionContrast, nodeColor=<rgba 1 0.9844 0 1>}
  NamedRerouteDeclaration(None=FunctionInput(Preview=MaterialFunctionCall!(TechArtMask) {materialFunction=/Game/Global/Materials/Character_Parent_2023/Glass_Parent_2023/MF_Glass}.Result) {inputName=In_Glass, inputType=FunctionInput_MaterialAttributes}) {name=IN_Glass, nodeColor=<rgba 0 1 0.9297 1>}
  n25 = FunctionInput() {inputName=Material, inputType=FunctionInput_MaterialAttributes}
  NamedRerouteDeclaration(None=GetMaterialAttributes(None=NamedRerouteDeclaration(None=@n25) {name=IN_Material, nodeColor=<rgba 1 0.5391 0 1>}) {attributeGetTypes [#3, #2]}.EmissiveColor) {name=IN_EmissiveColor, nodeColor=<rgba 0 1 0.1094 1>}
  n26 = NamedRerouteUsage()
  n27 = TextureSample(UVs=MaterialFunctionCall(X=Multiply(A=@n26) {constB=0.5}, Y=Constant() {r=0.5}) {materialFunction=&MakeFloat2}.Result) {
   samplerSource=SSM_Clamp_WorldGroupSettings
   texture=&/T_TurboFueled_Gradient
  }
  n28 = ScalarParameter() {defaultValue=1.2, group=BucketKick, parameterName=GradientExponent, sortPriority=50}
  NamedRerouteDeclaration(None=Multiply(A=Power(Base=@n27.RGB, Exp=@n28), B=ScalarParameter() {defaultValue=1, group=BucketKick, parameterName=EmissiveStrength, sortPriority=51})) {name=EmissiveOUT_2, nodeColor=<rgba 1 0.0938 0 1>}
  n29 = NamedRerouteUsage()
  n30 = NamedRerouteUsage()
  NamedRerouteDeclaration(None=Add(A=Multiply(A=@n29) {constB=0.5}, B=Multiply(A=@n30) {constB=0.5})) {name=EmissiveOutTotal, nodeColor=<rgba 0 0.4844 1 1>}
  n31 = NamedRerouteUsage()
  n32 = DotProduct(A=CameraVectorWS(), B=PixelNormalWS())
  n33 = NamedRerouteDeclaration(None=Power(Base=Saturate(None=@n32), Exp=ScalarParameter() {defaultValue=3, group=BucketKick, parameterName=FresnelExponent, sortPriority=43})) {name=FresenlValue_1, nodeColor=<rgba 0.7344 0 1 1>}
  NamedRerouteDeclaration(None=Saturate(None=Add(A=@n31, B=@n33))) {name=GradientUVX_2, nodeColor=<rgba 0 1 0.0625 1>}
  n34 = Power(
   Base=NamedRerouteUsage()
   Exp=Add(A=ScalarParameter() {defaultValue=0.5, group=BucketKick, parameterName=DistortionNoiseExponent, sortPriority=14}) {constB=0.1}
  )
  n35 = ScalarParameter() {defaultValue=0.2, group=BucketKick, parameterName=DistortionNoiseStrength, sortPriority=10}
  NamedRerouteDeclaration(None=Multiply(A=ConstantBiasScale(None=@n34) {bias=-0.5, scale=2}, B=Add(A=@n35) {constB=0.3})) {name=Distortion_1, nodeColor=<rgba 1 0.7031 0 1>}
  n36 = VertexInterpolator(VS=LocalPosition() {includedOffsets=ExcludeOffsets, localOrigin=InstancePreSkinning}.XYZ)
  n37 = NamedRerouteUsage()
  n38 = Time() {period=2}
  n39 = VectorParameter() {group=BucketKick, parameterName=DistortionNoiseSpeed, sortPriority=12, defaultValue=<rgba 0 0 -11 0>}
  n40 = Multiply(A=@n39.RGB) {constB=1.33}
  n41 = Add(
   A=Add(A=LinearInterpolate(A=@n36.PS, B=@n37) {constAlpha=0.1}, B=Multiply(A=@n38, B=@n40))
   B=Add(A=NamedRerouteUsage(), B=NamedRerouteUsage())
  )
  n42 = TextureObject() {samplerType=SAMPLERTYPE_LinearColor, texture=&T_TurboFueled_Body_FX}
  n43 = TextureSample(UVs=ComponentMask(None=Multiply(B=@n41) {constA=0.05}) {b=true, r=true}, Tex=@n42)
  NamedRerouteDeclaration(None=Saturate(None=ComponentMask(None=@n43.RGB) {b=true, g=true})) {name=DistortionNoiseSample_1, nodeColor=<rgba 1 0 0.5391 1>}
  NamedRerouteDeclaration(None=FunctionInput() {bUsePreviewValueAsDefault=true, inputName=Mask, sortPriority=2, previewValue=<1 0 0 1>}) {name=TechArtMask, nodeColor=<rgba 0.0001 1 0.9763 1>}
  n44 = TextureSample(UVs=MaterialFunctionCall(X=NamedRerouteUsage(), Y=Constant() {r=0.5}) {materialFunction=&MakeFloat2}.Result) {
   samplerSource=SSM_Clamp_WorldGroupSettings
   texture=&T_TurboFueled_Gradient
  }
  n45 = ScalarParameter() {defaultValue=1, group=BucketKick, parameterName=GradientExponent, sortPriority=50}
  NamedRerouteDeclaration(None=Multiply(A=Power(Base=@n44.RGB, Exp=@n45), B=ScalarParameter() {defaultValue=1.2, group=BucketKick, parameterName=EmissiveStrength, sortPriority=51})) {name=EmissiveOUT_1, nodeColor=<rgba 1 0.0938 0 1>}
 ]}
