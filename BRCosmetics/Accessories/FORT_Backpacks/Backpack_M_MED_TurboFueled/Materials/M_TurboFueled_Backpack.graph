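# Master material for the TurboFueled backpack cosmetic, built on the 2023 character parent PreFX and PostFX functions with an expression-driven shading model.
# PreFX outputs, including tech art mask and Fresnel, are shared through named reroutes.
# A car clear-coat layer, with its own coat normal output, and a glass layer with thin film feed the TurboFueled backpack FX function.
# That result passes through PostFX as material attributes.
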Material /BRCosmetics/Accessories/FORT_Backpacks/Backpack_M_MED_TurboFueled/Materials/M_TurboFueled_Backpack
graph {
 properties {
  bHasPixelAnimation=true
  bUseMaterialAttributes=true
  bUsedWithMeshDeformer=true
  bUsedWithMorphTargets=true
  bUsedWithSkeletalMesh=true
  pixelDepthOffsetMode=PDOM_Legacy
  shadingModel=MSM_FromMaterialExpression
  usedShadingModels=Default Lit | Subsurface | Clear Coat | Subsurface Profile | Cloth
 }
 nodes [
  n1 = MaterialFunctionCall!(Mask=NamedRerouteUsage(), Fresnel=NamedRerouteUsage(), In=NamedRerouteUsage()) {materialFunction=/Game/Global/MaterialFunctions/TechArt/MF_Advanced_CarClearCoat_v2}
  ClearCoatNormalCustomOutput!(None=@n1.ClearCoatNormal)
  n2 = MaterialFunctionCall!(TechArtMask=NamedRerouteUsage()) {materialFunction=&/MF_TurboFueled_Backpack_Glass}
  n3 = NamedRerouteUsage()
  n4 = NamedRerouteUsage()
  n5 = MaterialFunctionCall() {materialFunction=&MF_CharacterParent_2023_PreFX}
  MaterialFunctionCall(
   Material=MaterialFunctionCall!(
    In_Glass=MaterialFunctionCall!(Material Attributes=@n2.Result, Fresnel=@n3, Alpha=@n4) {materialFunction=/Game/Global/MaterialFunctions/TechArt/MF_ThinFilm_TechArt}.Result
    Material=@n1.Result
    Mask=NamedRerouteUsage()
   ) {materialFunction=/BRCosmetics/Characters/Player/Male/Medium/Bodies/M_MED_TurboFueled/FX/MF_TurboFueled_Backpack}.Result
   M=@n5.M
   Pre Skinned Local Position=@n5.M
   Distance Blend=@n5.M
   Fresnel=@n5.M
  ) {materialFunction=&MF_CharacterParent_2023_PostFX}
  NamedRerouteDeclaration(None=@n5.Tech Art Mask) {name=TechArtMask, nodeColor=<rgba 0 1 0.4141 1>}
  NamedRerouteDeclaration(None=@n5.Fresnel) {name=Fresnel, nodeColor=<rgba 1 0.0938 0 1>}
  NamedRerouteDeclaration(None=@n5.Result) {name=MaterialAttributes, nodeColor=<rgba 0 0.2266 1 1>}
 ]
}
